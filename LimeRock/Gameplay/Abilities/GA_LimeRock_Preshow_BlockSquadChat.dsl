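; LimeRock preshow gameplay ability that blocks squad chat emotes, squad quick chat and map marker placement for the activating pawn.
; On activation, and also on failed passive activation, it stores the pawn and blocks those tags on its Fort ability system component.
; On end, if the stored pawn is valid, it unblocks only the squad chat emote tag, not the quick chat or map marker tags.
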(event Ability|EventActivateAbility
  ;[0 0 0] block ability to emote
  (bind _output_get (Variables|Default|SetActivatingPawn (Ability|GetActivatingPawn)))
  (Tags|BlockAbilitiesWithTags (Ability|GetFortAbilitySystemComponent _output_get) (GameplayTags|MakeLiteralGameplayTagContainer "(GameplayTags=((TagName=\"Gameplay.Action.Player.Emote.SquadChat\"),(TagName=\"Gameplay.Action.SquadQuickChat\"),(TagName=\"Gamepad.ConfigCustom_Athena.CombatAndAthenaLTMAbilities.PlaceMapMarkerInWorld\"),(TagName=\"Gameplay.Action.Player.PlaceMapMarkerInWorld\")))")))

(event Ability|EventFailedToActivatePassiveAbility
  (params (ActorInfo :type Gameplay%20Ability%20Actor%20Info%20Structure))
  ;[0 0 0] block ability to emote
  (bind _output_get (Variables|Default|SetActivatingPawn (Ability|GetActivatingPawn)))
  (Tags|BlockAbilitiesWithTags (Ability|GetFortAbilitySystemComponent _output_get) (GameplayTags|MakeLiteralGameplayTagContainer "(GameplayTags=((TagName=\"Gameplay.Action.Player.Emote.SquadChat\"),(TagName=\"Gameplay.Action.SquadQuickChat\"),(TagName=\"Gamepad.ConfigCustom_Athena.CombatAndAthenaLTMAbilities.PlaceMapMarkerInWorld\"),(TagName=\"Gameplay.Action.Player.PlaceMapMarkerInWorld\")))")))

(event Ability|EventOnEndAbility
  (params (bWasCancelled :type Boolean))
  ;[0 0 0] unblock emote
  (bind _activatingpawn (Variables|Default|GetActivatingPawn)
    (:then
      (Tags|UnBlockAbilitiesWithTags (Ability|GetFortAbilitySystemComponent _activatingpawn) (GameplayTags|MakeLiteralGameplayTagContainer "(GameplayTags=((TagName=\"Gameplay.Action.Player.Emote.SquadChat\")))") false))
    (:else)))
