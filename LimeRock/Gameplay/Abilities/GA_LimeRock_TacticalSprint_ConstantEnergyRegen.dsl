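; Passive gameplay ability that grants constant tactical sprint energy regeneration in LimeRock.
; On activation it prints a debug string and starts adding energy to the tactical sprint energy component, using a scalable per-second rate and a source ID tag.
; On end it stops that source only if the activating pawn has the Granted.Athena.Energy.ConstantRegen tag.
; The failed passive activation event only calls the parent.
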(event Ability|EventActivateAbility
  ;[1 1 1] On Activate begin adding energy. On end remove the added energy.
  (Development|PrintString "EnergyRegenGranted")
  (bind _energycomponent (Default|GetTacticalSprintEnergyComponentBranching)
    (:Is%20Valid
      (Energy|StartAddingEnergy _energycomponent (ScalableFloat|GetValueatLevel (Variables|Default|GetRow_EnergyPerSecond)) (Variables|Default|GetT_EnergySourceID) 0))))

(event Ability|EventOnEndAbility
  (params (bWasCancelled :type Boolean))
  ;[1 1 1] On Activate begin adding energy. On end remove the added energy.
  (bind _energycomponent (Default|GetTacticalSprintEnergyComponentBranching)
    (:Is%20Valid
      (if (GameplayTags|HasMatchingGameplayTag (Ability|GetActivatingPawn) "(TagName=\"Granted.Athena.Energy.ConstantRegen\")")
        (Energy|StopAddingEnergy _energycomponent (Variables|Default|GetT_EnergySourceID) 0)))))

(event Ability|EventFailedToActivatePassiveAbility
  (params (ActorInfo :type Gameplay%20Ability%20Actor%20Info%20Structure))
  ;[1 1 1] Must implement this given it is a passive ability.
  (|CallParentFunction))
